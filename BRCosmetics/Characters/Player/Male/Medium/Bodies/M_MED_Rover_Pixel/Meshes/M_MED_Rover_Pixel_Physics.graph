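# Physics asset for the Rover Pixel male medium player body cosmetic, covering the dynamic bones of its holsters and shirt flap.
# A large spine_03 capsule is the root body, with small master capsules for each holster and the flap joined to it by angularly free constraints.
# Dangling bodies swing mainly on one limited axis, with tight twist and second swing.
# The left holster branches from L1 into L2 and L3, while the right has one slightly looser segment.
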PhysicsAsset /BRCosmetics/Characters/Player/Male/Medium/Bodies/M_MED_Rover_Pixel/Meshes/M_MED_Rover_Pixel_Physics
bodies {
 Dyn_Holsters_L1 [{shapeType=Capsule, length=2, radius=3, shapeName=Dyn_Holsters_L1_capsule, center=<4.5519 -1.8389 0>, rotation=<rot 90 0 0>}]
 Dyn_Holsters_L2 [{shapeType=Capsule, length=1, radius=1, shapeName=Dyn_Holsters_L2_capsule, center=<1 0 0>, rotation=<rot 90 0 0>}]
 Dyn_Holsters_L3 [{shapeType=Capsule, length=1, radius=1, shapeName=Dyn_Holsters_L3_capsule, center=<1 0 0>, rotation=<rot 90 0 0>}]
 Dyn_Holsters_R1 [{shapeType=Capsule, length=2, radius=2, shapeName=Dyn_Holsters_R1_capsule, center=<3 0 0>, rotation=<rot 45.0001 0 0>}]
 Dyn_Master_Holsters_L [{shapeType=Capsule, length=0.5, radius=0.5, shapeName=Dyn_Master_Holsters_L_capsule}]
 Dyn_Master_Holsters_R [{shapeType=Capsule, length=0.5, radius=0.5, shapeName=Dyn_Master_Holsters_R_capsule}]
 Dyn_Master_ShirtFlap [{shapeType=Capsule, length=0.5, radius=0.5, shapeName=Dyn_Master_ShirtFlap_capsule}]
 Dyn_ShirtFlap_1 [{shapeType=Capsule, length=1, radius=1.5, shapeName=Dyn_ShirtFlap_1_capsule, center=<1 0 0>}]
 spine_03 [{shapeType=Capsule, length=16.4612, radius=12.6, shapeName=spine_03_capsule, rotation=<rot 90 0 0>}]
}
constraints [
 {bone1Name=Dyn_ShirtFlap_1, bone2Name=Dyn_Master_ShirtFlap, swing1LimitDegrees=30, swing2LimitDegrees=1, twistLimitDegrees=1}
 {bone1Name=Dyn_Holsters_L1, bone2Name=Dyn_Master_Holsters_L, swing1LimitDegrees=30, swing2LimitDegrees=1, twistLimitDegrees=1}
 {bone1Name=Dyn_Holsters_L2, bone2Name=Dyn_Holsters_L1, swing1LimitDegrees=30, swing2LimitDegrees=1, twistLimitDegrees=1}
 {bone1Name=Dyn_Holsters_L3, bone2Name=Dyn_Holsters_L1, swing1LimitDegrees=30, swing2LimitDegrees=1, twistLimitDegrees=1}
 {bone1Name=Dyn_Holsters_R1, bone2Name=Dyn_Master_Holsters_R, swing1LimitDegrees=30, swing2LimitDegrees=10, twistLimitDegrees=5}
 {
  bone1Name=Dyn_Master_ShirtFlap
  bone2Name=spine_03
  swing1LimitDegrees=45
  swing1Motion=Free
  swing2LimitDegrees=45
  swing2Motion=Free
  twistLimitDegrees=45
  twistMotion=Free
 }
 {
  bone1Name=Dyn_Master_Holsters_L
  bone2Name=spine_03
  swing1LimitDegrees=45
  swing1Motion=Free
  swing2LimitDegrees=45
  swing2Motion=Free
  twistLimitDegrees=45
  twistMotion=Free
 }
 {
  bone1Name=Dyn_Master_Holsters_R
  bone2Name=spine_03
  swing1LimitDegrees=45
  swing1Motion=Free
  swing2LimitDegrees=45
  swing2Motion=Free
  twistLimitDegrees=45
  twistMotion=Free
 }
]
