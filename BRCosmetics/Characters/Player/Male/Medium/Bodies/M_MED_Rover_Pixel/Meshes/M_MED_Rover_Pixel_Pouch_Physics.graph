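# Physics asset for the pouch on the M_MED_Rover_Pixel male player cosmetic body.
# It has two capsule bodies: a tiny one on Dyn_Master_Pouch and a larger, offset and rotated one covering Dyn_Pouch1.
# A single constraint joins them with narrow swing limits and an almost locked twist, restricting how far the pouch can move relative to its master bone.
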PhysicsAsset /BRCosmetics/Characters/Player/Male/Medium/Bodies/M_MED_Rover_Pixel/Meshes/M_MED_Rover_Pixel_Pouch_Physics
bodies {
 Dyn_Master_Pouch [{shapeType=Capsule, length=0.505, radius=0.505, shapeName=Dyn_Master_Pouch_capsule}]
 Dyn_Pouch1 [{shapeType=Capsule, length=7, radius=4, shapeName=Dyn_Pouch1_capsule, center=<1.5 -5.4 0>, rotation=<rot 90 0 0>}]
}
constraints [{bone1Name=Dyn_Pouch1, bone2Name=Dyn_Master_Pouch, swing1LimitDegrees=20, swing2LimitDegrees=5, twistLimitDegrees=1}]
