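# Physics asset for the face accessory part of the Rover Pixel medium male player body, simulating a short hair bone chain on the head.
# Three capsule bodies cover the head, Dyn_Master_Hair and Dyn_Master_Hair1.
# The root hair bone rotates freely on swing and twist relative to the head, while the second bone has narrow swing limits and almost no twist relative to the first.
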PhysicsAsset /BRCosmetics/Characters/Player/Male/Medium/Bodies/M_MED_Rover_Pixel/Meshes/Parts/M_MED_Rover_Pixel_FaceAcc_Physics
bodies {
 Dyn_Master_Hair [{shapeType=Capsule, length=0.5, radius=0.5, shapeName=Dyn_Master_Hair_capsule}]
 Dyn_Master_Hair1 [{shapeType=Capsule, length=3, radius=2, shapeName=Dyn_Master_Hair1_capsule, center=<1 0 0>, rotation=<rot 90 0 0>}]
 head [{shapeType=Capsule, length=1, radius=1}]
}
constraints [
 {
  bone1Name=Dyn_Master_Hair
  bone2Name=head
  swing1LimitDegrees=45
  swing1Motion=Free
  swing2LimitDegrees=45
  swing2Motion=Free
  twistLimitDegrees=45
  twistMotion=Free
 }
 {bone1Name=Dyn_Master_Hair1, bone2Name=Dyn_Master_Hair, swing1LimitDegrees=30, swing2LimitDegrees=20, twistLimitDegrees=1}
]
